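# Sound cue for a planter being destroyed in a Season 29 Battle Pass snapshot quest.
# A random node picks one of three equally weighted destruction wave variations each time it plays.
# It routes through the quests sound class and borrows a supply-drop hidden-star ambient loop attenuation asset for spatial falloff.
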SoundCue /BattlePassS29_Quests/Audio/SFX/Quest_S29_snapshot_p04_SpawnedFleece/Quest_S29_snapshot_p04_Planter_Destroyed_Cue
root=Random_0
nodes [
 {id=Random_0, children [WavePlayer_0,WavePlayer_1,WavePlayer_2], properties {weights [1,1,1]}}
 {
  id=WavePlayer_0
  properties {soundWaveAssetPtr=&/S29_Quest04_Planter_Destroyed_01}
 }
 {
  id=WavePlayer_1
  properties {soundWaveAssetPtr=&/S29_Quest04_Planter_Destroyed_02}
 }
 {
  id=WavePlayer_2
  properties {soundWaveAssetPtr=&/S29_Quest04_Planter_Destroyed_03}
 }
]
properties {
 attenuationSettings=/Game/Athena/Sounds/SupplyDrops/Fort_HiddenStar_Ambient_Loop_Att
 maxDistance=2500
 soundClassObject=&Fort_Quests
}
