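# Sound cue for picking up a spawned fleece in a Season 29 Battle Pass snapshot quest.
# A mixer layers a randomly chosen, modulated variation from three fleece pickup waves with a shared reward-screen chest loot upgrade UI sound.
# It plays in the quests sound class at reduced volume, borrowing attenuation from a supply drop hidden star ambient loop.
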SoundCue /BattlePassS29_Quests/Audio/SFX/Quest_S29_snapshot_p04_SpawnedFleece/Quest_S29_snapshot_p04_SpawnedFleece_Pickup_Cue
root=Mixer_0
nodes [
 {id=Mixer_0, children [Modulator_0,WavePlayer_4], properties {inputVolume [1,1]}}
 {id=Modulator_0, children [Random_0]}
 {id=Random_0, children [WavePlayer_0,WavePlayer_1,WavePlayer_2], properties {weights [1,1,1]}}
 {
  id=WavePlayer_0
  properties {soundWaveAssetPtr=&/S29_Quest04_SpawnedFleece_Pickup_01}
 }
 {
  id=WavePlayer_1
  properties {soundWaveAssetPtr=&/S29_Quest04_SpawnedFleece_Pickup_02}
 }
 {
  id=WavePlayer_2
  properties {soundWaveAssetPtr=&/S29_Quest04_SpawnedFleece_Pickup_03}
 }
 {id=WavePlayer_4, properties {soundWaveAssetPtr=/Game/Sounds/Fort_UI_Sounds/Rewards/Fort_RewardScreen_TeamScore_ChestLootUpgrade}}
]
properties {
 attenuationSettings=/Game/Athena/Sounds/SupplyDrops/Fort_HiddenStar_Ambient_Loop_Att
 maxDistance=2500
 soundClassObject=&Fort_Quests
 volumeMultiplier=0.7
}
